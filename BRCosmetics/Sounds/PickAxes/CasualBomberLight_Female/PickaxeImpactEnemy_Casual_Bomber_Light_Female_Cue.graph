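# Sound cue for the Casual Bomber Light Female pickaxe hitting an enemy, using pickaxe impact attenuation and melee concurrency.
# A quality-level node selects branches with differing variation counts, built from randomized impact and vocalisation waves behind modulators and a short fixed delay.
# In three quality slots, a weighted random favors impact alone over impact mixed with a quieter vocalisation.
# The remaining slot always mixes both from the smallest pools.
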SoundCue /BRCosmetics/Sounds/PickAxes/CasualBomberLight_Female/PickaxeImpactEnemy_Casual_Bomber_Light_Female_Cue
root=QualityLevel_0
nodes [
 {id=Delay_0, children [Modulator_0], properties {delayMax=0.05, delayMin=0.05}}
 {id=Delay_1, children [Modulator_1], properties {delayMax=0.05, delayMin=0.05}}
 {id=Delay_2, children [Modulator_2], properties {delayMax=0.05, delayMin=0.05}}
 {id=Delay_3, children [Modulator_3], properties {delayMax=0.05, delayMin=0.05}}
 {id=Delay_4, children [Modulator_4], properties {delayMax=0.05, delayMin=0.05}}
 {id=Mixer_0, children [Delay_1,Delay_0], properties {inputVolume [1,0.1]}}
 {id=Mixer_1, children [Delay_2,Delay_4], properties {inputVolume [1,0.6]}}
 {id=Mixer_2, children [Random_10,Random_11], properties {inputVolume [1,0.6]}}
 {id=Modulator_0, children [Random_2]}
 {id=Modulator_1, children [Random_0]}
 {id=Modulator_2, children [Random_9]}
 {id=Modulator_3, children [Mixer_2]}
 {id=Modulator_4, children [Random_8]}
 {id=QualityLevel_0, children [Random_5,Random_7,Delay_3,Random_7]}
 {
  id=Random_0
  children [WavePlayer_0, WavePlayer_6, WavePlayer_7, WavePlayer_8, WavePlayer_9, WavePlayer_10]
  properties {weights [1,1,1,1,1,1]}
 }
 {id=Random_10, children [WavePlayer_36, WavePlayer_37], properties {weights [1,1]}}
 {id=Random_11, children [WavePlayer_38, WavePlayer_39], properties {weights [1,1]}}
 {
  id=Random_2
  children [WavePlayer_1, WavePlayer_2, WavePlayer_3, WavePlayer_4, WavePlayer_5, WavePlayer_12, WavePlayer_13, WavePlayer_14]
  properties {weights [1,1,1,1,1,1,1,1]}
 }
 {id=Random_5, children [Delay_1,Mixer_0], properties {bRandomizeWithoutReplacement=false, weights [1,0.4]}}
 {id=Random_7, children [Delay_2,Mixer_1], properties {bRandomizeWithoutReplacement=false, weights [1,0.4]}}
 {id=Random_8, children [WavePlayer_30, WavePlayer_31, WavePlayer_32], properties {weights [1,1,1]}}
 {id=Random_9, children [WavePlayer_33, WavePlayer_34, WavePlayer_35], properties {weights [1,1,1]}}
 {
  id=WavePlayer_0
  properties {soundWaveAssetPtr=&/PickAxe_Casual_Bomber_Light_Female_Impact_01}
 }
 {
  id=WavePlayer_1
  properties {soundWaveAssetPtr=&/PickAxe_Casual_Bomber_Light_Female_Impact_Vocalisation_01}
 }
 {
  id=WavePlayer_10
  properties {soundWaveAssetPtr=&/PickAxe_Casual_Bomber_Light_Female_Impact_06}
 }
 {
  id=WavePlayer_12
  properties {soundWaveAssetPtr=&/PickAxe_Casual_Bomber_Light_Female_Impact_Vocalisation_06}
 }
 {
  id=WavePlayer_13
  properties {soundWaveAssetPtr=&/PickAxe_Casual_Bomber_Light_Female_Impact_Vocalisation_07}
 }
 {
  id=WavePlayer_14
  properties {soundWaveAssetPtr=&/PickAxe_Casual_Bomber_Light_Female_Impact_Vocalisation_08}
 }
 {
  id=WavePlayer_2
  properties {soundWaveAssetPtr=&/PickAxe_Casual_Bomber_Light_Female_Impact_Vocalisation_02}
 }
 {
  id=WavePlayer_3
  properties {soundWaveAssetPtr=&/PickAxe_Casual_Bomber_Light_Female_Impact_Vocalisation_03}
 }
 {
  id=WavePlayer_30
  properties {soundWaveAssetPtr=&PickAxe_Casual_Bomber_Light_Female_Impact_Vocalisation_01}
 }
 {
  id=WavePlayer_31
  properties {soundWaveAssetPtr=&PickAxe_Casual_Bomber_Light_Female_Impact_Vocalisation_03}
 }
 {
  id=WavePlayer_32
  properties {soundWaveAssetPtr=&PickAxe_Casual_Bomber_Light_Female_Impact_Vocalisation_02}
 }
 {
  id=WavePlayer_33
  properties {soundWaveAssetPtr=&PickAxe_Casual_Bomber_Light_Female_Impact_01}
 }
 {
  id=WavePlayer_34
  properties {soundWaveAssetPtr=&/PickAxe_Casual_Bomber_Light_Female_Impact_02}
 }
 {
  id=WavePlayer_35
  properties {soundWaveAssetPtr=&/PickAxe_Casual_Bomber_Light_Female_Impact_03}
 }
 {
  id=WavePlayer_36
  properties {soundWaveAssetPtr=&PickAxe_Casual_Bomber_Light_Female_Impact_01}
 }
 {
  id=WavePlayer_37
  properties {soundWaveAssetPtr=&PickAxe_Casual_Bomber_Light_Female_Impact_02}
 }
 {
  id=WavePlayer_38
  properties {soundWaveAssetPtr=&PickAxe_Casual_Bomber_Light_Female_Impact_Vocalisation_01}
 }
 {
  id=WavePlayer_39
  properties {soundWaveAssetPtr=&PickAxe_Casual_Bomber_Light_Female_Impact_Vocalisation_02}
 }
 {
  id=WavePlayer_4
  properties {soundWaveAssetPtr=&/PickAxe_Casual_Bomber_Light_Female_Impact_Vocalisation_04}
 }
 {
  id=WavePlayer_5
  properties {soundWaveAssetPtr=&/PickAxe_Casual_Bomber_Light_Female_Impact_Vocalisation_05}
 }
 {
  id=WavePlayer_6
  properties {soundWaveAssetPtr=&PickAxe_Casual_Bomber_Light_Female_Impact_02}
 }
 {
  id=WavePlayer_7
  properties {soundWaveAssetPtr=&PickAxe_Casual_Bomber_Light_Female_Impact_03}
 }
 {
  id=WavePlayer_8
  properties {soundWaveAssetPtr=&/PickAxe_Casual_Bomber_Light_Female_Impact_04}
 }
 {
  id=WavePlayer_9
  properties {soundWaveAssetPtr=&/PickAxe_Casual_Bomber_Light_Female_Impact_05}
 }
]
properties {
 attenuationSettings=/Game/Sounds/Attenuation/Fort_Pickaxe_Impact_Att
 audioPropertiesSheet=/BRCosmetics/Sounds/Settings/PropertySheets/Pickaxe/APS_Cosmetic_Pickaxe_Impact
 bPrimeOnLoad=true
 maxDistance=4300
 soundClassObject=&Fort_WeaponMelee_PSM
 volumeMultiplier=0.8
 concurrencySet [&Player_Melee_Weapon_Conc]
}
